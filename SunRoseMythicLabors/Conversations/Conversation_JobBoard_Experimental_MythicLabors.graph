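# Conversation database for an experimental Mythic Labors job board offering Footrace and Waxen Wings quests.
# Each job's entry point leads to a brief with cancel and slotted quest grant choices, though Wings' opening speech uses the Footrace brief.
# Grants require the job's service and no conflicting urgent, experimental or bounty quest.
# Grant side effects play feedback, record history, track the user, apply an instigator effect and deactivate the board.
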ConversationDatabase /SunRoseMythicLabors/Conversations/Conversation_JobBoard_Experimental_MythicLabors
entry_points [
 {tag=Conversation.Entry.JobBoard.Footrace, destinations [#1]}
 {tag=Conversation.Entry.JobBoard.Wings, destinations [#2]}
]
nodes [
 ConversationEntryPointNode_1 {
  id=#2
  class=ConversationEntryPointNode
  outputs [#3]
  properties {entryTag {tagName=Conversation.Entry.JobBoard.Wings}}
 }
 Speech_C_1 {
  id=#3
  class=Speech_C
  outputs [#4, #5]
  properties {briefWidgetClassOverride=/SunRoseMythicLabors/Items/Widgets/FortFootraceJobBrief.FortFootraceJobBrief_C, nodeName=""}
 }
 GrantSlottedQuest_C_1 {
  id=#5
  class=GrantSlottedQuest_C
  outputs []
  properties {
   bSharedQuestLimitsEnforced=false
   bShouldOverrideSpeakerText=false
   briefWidgetClassOverride=/SunRoseMythicLabors/Items/Widgets/FortWaxenWingsJobBrief.FortWaxenWingsJobBrief_C
   softTaskIcon=/Game/UI/Foundation/Textures/Icons/Checks/T-Icon-Check-128
   optionalFilterCategoryTag {tagName=Tandem.Service.JobTypeExperimental.Wings}
  }
  sub_nodes [
   SideEffect_PlayServiceUsedEffects_C_1 {class=SideEffect_PlayServiceUsedEffects_C}
   SideEffect_Server_RecordEventToHistoryData_C_1 {class=SideEffect_Server_RecordEventToHistoryData_C, properties {event Type {tagName=Conversation.Event.GrantedQuest}}}
   SideEffect_Server_TrackJobBoardUser_C_1 {class=SideEffect_Server_TrackJobBoardUser_C}
   SideEffect_Server_JobBoardExperimental_ApplyInstigatorGE_C_1 {class=SideEffect_Server_JobBoardBriefing_ApplyInstigatorGE_C}
   SideEffect_Server_JobBoardExperimental_DeactivateOnUse_C_1 {class=SideEffect_Server_JobBoardBriefing_DeactivateOnUse_C}
   ConversationChoiceNode_2 {class=ConversationChoiceNode, properties {defaultChoiceDisplayText=WAXEN WINGS}}
   FortConversationRequirement_HasService_0 {class=FortConversationRequirement_HasService, properties {serviceTag {tagName=Tandem.Service.JobTypeExperimental.Wings}}}
   FortConversationRequirement_ControllerRequirement_38 {
    class=FortConversationRequirement_ControllerRequirement
    properties {
     nodeName=Require: Has Urgent Quest
     requirement {
      class=FortControllerRequirement_HasUrgentQuest
      properties {
       bInverseCondition=true
       bRequireExactMatch=false
       acceptedUrgentEventTags {gameplayTags [{tagName=SpecialEvent.HuntingPlayer.IsHunter}, {tagName=SpecialEvent.JobBoard.Experimental}]}
      }
     }
    }
   }
   FortConversationRequirement_ControllerRequirement_39 {
    class=FortConversationRequirement_ControllerRequirement
    properties {
     nodeName=Require: Has GameplayTag
     requirement {
      class=FortControllerRequirement_HasOwnedTag
      properties {bInverseCondition=true, ownedTag {tagName=Granted.Athena.JobBoardExperimental.QuestActive}}
     }
    }
   }
   FortConversationRequirement_ControllerRequirement_40 {
    class=FortConversationRequirement_ControllerRequirement
    properties {
     nodeName=Require: Has GameplayTag
     requirement {class=FortControllerRequirement_HasOwnedTag, properties {bInverseCondition=true, ownedTag {tagName=Granted.Athena.BountyBoard.QuestActive}}}
    }
   }
  ]
 }
 Continue_C_0 {
  id=#4
  class=Continue_C
  properties {softTaskIcon=&T_UI_NPC_Back}
  sub_nodes [ConversationChoiceNode_6 {
    class=ConversationChoiceNode
    properties {defaultChoiceDisplayText=CANCEL, choiceTags {gameplayTags [{tagName=Conversation.Choice.CustomSortPriority.Highest}]}}
   }]
 }
 ConversationEntryPointNode_5 {
  id=#1
  class=ConversationEntryPointNode
  outputs [#6]
  properties {entryTag {tagName=Conversation.Entry.JobBoard.Footrace}}
 }
 Speech_C_0 {
  id=#6
  class=Speech_C
  outputs [#7, #8]
  properties {briefWidgetClassOverride=&FortFootraceJobBrief_C, nodeName=""}
 }
 GrantSlottedQuest_C_0 {
  id=#8
  class=GrantSlottedQuest_C
  outputs []
  properties {
   bSharedQuestLimitsEnforced=false
   bShouldOverrideSpeakerText=false
   briefWidgetClassOverride=&FortFootraceJobBrief_C
   softTaskIcon=&T-Icon-Check-128
   optionalFilterCategoryTag {tagName=Tandem.Service.JobTypeExperimental.Footrace}
  }
  sub_nodes [
   SideEffect_PlayServiceUsedEffects_C_0 {class=SideEffect_PlayServiceUsedEffects_C}
   SideEffect_Server_RecordEventToHistoryData_C_0 {class=SideEffect_Server_RecordEventToHistoryData_C, properties {event Type {tagName=Conversation.Event.GrantedQuest}}}
   SideEffect_Server_TrackJobBoardUser_C_0 {class=SideEffect_Server_TrackJobBoardUser_C}
   SideEffect_Server_JobBoardExperimental_ApplyInstigatorGE_C_0 {class=SideEffect_Server_JobBoardBriefing_ApplyInstigatorGE_C}
   SideEffect_Server_JobBoardExperimental_DeactivateOnUse_C_0 {class=SideEffect_Server_JobBoardBriefing_DeactivateOnUse_C}
   ConversationChoiceNode_0 {class=ConversationChoiceNode, properties {defaultChoiceDisplayText=FOOTRACE}}
   FortConversationRequirement_HasService_1 {class=FortConversationRequirement_HasService, properties {serviceTag {tagName=Tandem.Service.JobTypeExperimental.Footrace}}}
   FortConversationRequirement_ControllerRequirement_0 {
    class=FortConversationRequirement_ControllerRequirement
    properties {
     nodeName=Require: Has Urgent Quest
     requirement {
      class=FortControllerRequirement_HasUrgentQuest
      properties {
       bInverseCondition=true
       bRequireExactMatch=false
       acceptedUrgentEventTags {gameplayTags [{tagName=SpecialEvent.HuntingPlayer.IsHunter}, {tagName=SpecialEvent.JobBoard.Experimental}]}
      }
     }
    }
   }
   FortConversationRequirement_ControllerRequirement_1 {
    class=FortConversationRequirement_ControllerRequirement
    properties {
     nodeName=Require: Has GameplayTag
     requirement {
      class=FortControllerRequirement_HasOwnedTag
      properties {bInverseCondition=true, ownedTag {tagName=Granted.Athena.JobBoardExperimental.QuestActive}}
     }
    }
   }
   FortConversationRequirement_ControllerRequirement_2 {
    class=FortConversationRequirement_ControllerRequirement
    properties {
     nodeName=Require: Has GameplayTag
     requirement {class=FortControllerRequirement_HasOwnedTag, properties {bInverseCondition=true, ownedTag {tagName=Granted.Athena.BountyBoard.QuestActive}}}
    }
   }
  ]
 }
 Continue_C_5 {
  id=#7
  class=Continue_C
  properties {softTaskIcon=&T_UI_NPC_Back}
  sub_nodes [ConversationChoiceNode_1 {
    class=ConversationChoiceNode
    properties {defaultChoiceDisplayText=CANCEL, choiceTags {gameplayTags [{tagName=Conversation.Choice.CustomSortPriority.Highest}]}}
   }]
 }
]
